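# Physics asset for the Scarab vehicle cosmetic body, giving its mesh a simple collision and physics shape.
# A single body on the frame bone holds one large capsule, offset and rotated from the bone's origin.
# It has no constraints, so the asset defines no jointed or ragdoll parts and approximates the vehicle body with that one primitive.
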PhysicsAsset /VehicleCosmetics/Bodies/Scarab/Meshes/SM_Scarab_PhysicsAsset
bodies {frame [{shapeType=Capsule, length=207.9263, radius=138.8352, center=<33.4986 -95.8907 0.996>, rotation=<rot 0 -84.8448 -88.2584>}]}
constraints []
